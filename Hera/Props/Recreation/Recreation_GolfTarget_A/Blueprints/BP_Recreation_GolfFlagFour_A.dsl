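; Golf flag prop Blueprint that places or removes its Hole component during construction.
; After calling the parent construction script, if SpawnHole is set it line traces vertically through the actor's location using the landscape and water profile.
; It moves Hole slightly above the hit location without checking that the trace hit; otherwise it destroys the component.
; BeginPlay, overlap and Tick events are empty.
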(event EventBeginPlay)

(event Collision|EventActorBeginOverlap
  (params (OtherActor :type Actor%20Object%20Reference)))

(event EventTick
  (params (DeltaSeconds :type Float%20%28single-precision%29)))

(fn ConstructionScript
  (|CallParentFunction)
  (if (Variables|Default|GetSpawnHole)
    (bind _outhit (Collision|LineTraceByProfile (+ (Transformation|GetActorLocation) "0.000000,0.000000,100.000000") (+ (Transformation|GetActorLocation) "0.000000,0.000000,-200.000000") "FortLandscapeAndWater"))
    (Transformation|SetWorldLocation (Variables|Default|GetHole) (+ (.Location (Collision|BreakHitResult _outhit)) "0.000000,0.000000,2.500000") false false)
    (else
      (Components|DestroyComponent (Variables|Default|GetHole)))))
